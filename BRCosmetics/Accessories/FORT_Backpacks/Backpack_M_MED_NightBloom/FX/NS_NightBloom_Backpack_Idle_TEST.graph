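# Test Niagara system for the idle effect of the NightBloom backpack cosmetic: a blue glow plus a point light.
# Glow bursts 18 persistent sprites in three partitions, placed on rings of differing size and offset, oriented by skeletal mesh sampling.
# Their alpha follows a clamped power of the camera-to-axis dot product, with separate gameplay and front-end scaling.
# Light drives a light renderer with blue color multiplied by a system time-of-day color.
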NiagaraSystem /BRCosmetics/Accessories/FORT_Backpacks/Backpack_M_MED_NightBloom/FX/NS_NightBloom_Backpack_Idle_TEST
dependencies {
 usedDataInterfaces [&NiagaraDataInterfaceCamera, &NiagaraDataInterfaceSkeletalMesh, &NiagaraDataInterfaceCurve]
 usedDynamicInputs [
  &FloatFromCurve
  &Subtract_Float
  /Game/Effects/Niagara/DynamicInputs/TimeOfDay/DaySequenceTimeOfDayColor
 ]
 usedModules [
  &self:SystemSpawnScript.SystemScriptSource.SystemScriptGraph.NiagaraNodeAssignment_0.SetVariables_#1
  &SystemState
  &self:SystemSpawnScript.SystemScriptSource.SystemScriptGraph.NiagaraNodeAssignment_2.SetVariables_#2
  &self:SystemSpawnScript.SystemScriptSource.SystemScriptGraph.NiagaraNodeAssignment_3.SetVariables_#3
 ]
}
emitters {
 Glow {
  emitterUpdateScript [
   &EmitterState(Life Cycle Mode=::NewEnumerator1, Loop Behavior=::NewEnumerator1, Loop Duration Mode=::NewEnumerator1, UseLoopDelay=false)
   &SpawnBurst_Instantaneous(Spawn Count=18, Use Spawn Probability=false, UseLoopCountLimit=false)
   &SampleSkeletalMesh(
    AdditionalYawPitchRoll=true
    Mesh Sampling Type:ENiagaraSkelSamplingModeFull=::NewEnumerator5
    Sampled Orientation Additional Yaw / Pitch / Roll:Vector3f=<0 90 90>
    Skeletal Mesh:NiagaraDataInterfaceSkeletalMesh=$System.Skeletal Mesh
    Socket Sampling Mode:ENiagaraSocketSamplingMode=::NewEnumerator5
   )
  ]
  particleSpawnScript [
   &PartitionParticles(Number Of Partitions=3)
   &InitializeParticle(
    Color=&MakeCustomLinearColorFromBool(Bool=$User.bIsFrontEndPreview, True Color=<rgba 0.4345 11.862 25 0.05>, False Color=<rgba 0.6357 13.887 25 0.05>)
    Sprite Size Mode=::NewEnumerator3
    Uniform Sprite Size=&Multiply_Float(
     A=&MergeHelper_FloatSwitch(0=10, 1=6, 2=16, 3=?, Index=$Particles.PartitionParticles.Partition, Num=::NewEnumerator1)
     B=1.8
    )
    UsePositionOffset=false
    Write Material Random=true
   )
  ]
  particleUpdateScript [
   &ParticleState(Kill Particles When Lifetime Has Elapsed=false, Let Infinitely Lived Particles Die When Emitter Deactivates=true)
   &ShapeLocation(
    Additional Quaternion Rotation=true
    Additional Yaw / Pitch / Roll=&MergeHelper_VectorSwitch(0=<0 0 0>, 1=<-0.056 0 0>, 2=<-0.115 0 0>, 3=?, Index=$Particles.PartitionParticles.Partition, Num=::NewEnumerator1)
    MaskSpawnGroup=false
    Offset=&MergeHelper_VectorSwitch(0=<0 -3 0>, 1=<0 -4 0>, 2=<0 -2 0>, 3=?, Index=$Particles.PartitionParticles.Partition, Num=::NewEnumerator1)
    Offset Mode=::NewEnumerator0
    Override Seed=false
    Ring / Disc Distribution Mode:ENiagaraSphereDistributionMode=::NewEnumerator1
    Ring Radius=&MergeHelper_FloatSwitch(0=14, 1=17, 2=8.5559, 3=?, Index=$Particles.PartitionParticles.Partition, Num=::NewEnumerator1)
    Rotation Mode=::NewEnumerator3
    Rotation Quaternion=$Emitter.SampleSkeletalMesh.SampledRotation
    Shape Primitive=::NewEnumerator4
    U Position=&ReturnNormalizedExecIndex(Normalized Index Scale=&MergeHelper_FloatSwitch(0=0.945, 1=0.95, 2=0.945, 3=?, Index=$Particles.PartitionParticles.Partition, Num=::NewEnumerator1))
    UseOwnerScale=true
   )
   &ScaleColor(
    Scale Alpha=&Multiply_Float(
     A=&Multiply_Float(A=1, B=1)
     B=&ClampFloat(
      Float=&Power(
       a=&Abs_Float(Float:NiagaraFloat=&DotProduct(Vector A=&ReturnCameraVector(Camera Vectors=::NewEnumerator1, Coordinate Space=::Local), Vector B=<0 1 0>))
       b=8
      )
      Min=&Multiply_Float(
       A=&MergeHelper_FloatSwitch(0=0.02, 1=0.04, 2=0.03, 3=?, Index=$Particles.PartitionParticles.Partition, Num=::NewEnumerator1)
       B=&FN_FrontEndFloatModifier(Gameplay=1, Front End=1)
      )
      Max=&Multiply_Float(
       A=&MergeHelper_FloatSwitch(0=0.3, 1=0.1, 2=0.8, 3=?, Index=$Particles.PartitionParticles.Partition, Num=::NewEnumerator1)
       B=&FN_FrontEndFloatModifier(Gameplay=0.4, Front End=0.2)
      )
     )
    )
    ScaleA=true
    ScaleRGB=false
   )
  ]
  rendererClasses [&NiagaraSpriteRendererProperties]
  renderers [{rendererClass=&NiagaraSpriteRendererProperties}]
 }
 Light {
  emitterUpdateScript [
   &EmitterState(
    Inactive Response=::NewEnumerator1
    Life Cycle Mode=::NewEnumerator1
    Loop Behavior=::NewEnumerator1
    Loop Duration Mode=::NewEnumerator1
    UseLoopDelay=false
   )
   &SpawnBurst_Instantaneous(Use Spawn Probability=false, UseLoopCountLimit=false)
  ]
  particleSpawnScript [&InitializeParticle(
    Color=&MultiplyLinearColor(Linear Color=<rgba 0.7552 0.7552 0.7552 0.3>, Scale Factor=<1 1 1 1>)
    Lifetime=&RandomRangeFloat(Minimum=0.6, Maximum=1)
    Sprite Size=<8 70>
    Sprite Size Mode=::NewEnumerator4
    UsePositionOffset=false
    Write Material Random=true
   )]
  particleUpdateScript [
   &ParticleState(Kill Particles When Lifetime Has Elapsed=false)
   &FN_Light_Attributes(
    Color=&Multiply_LinearColorByFloat(
     Linear Color=&MultiplyLinearColor(Linear Color=<rgba 0 0.6136 1.5 1>, Scale Factor=&MakeVector4_FromLinearColor(LinearColor=$System.TOD))
     Float=2
    )
    Create Color Attribute=true
    Create Exponent Attribute=true
    Create Position Attribute=true
    Create Radius Attribute=true
    Exponent=2
    Position=&AddVectorToPosition(Position=$Particles.Position, Vector=<0 -8 0>)
    Radius=75
   )
  ]
  rendererClasses [&NiagaraLightRendererProperties]
  renderers [{rendererClass=&NiagaraLightRendererProperties, rendererIndex=0}]
 }
}
summary {
 systemName=NS_NightBloom_Backpack_Idle_TEST
 emitters [
  {emitterName=Glow, rendererClasses [&NiagaraSpriteRendererProperties]}
  {emitterName=Light, rendererClasses [&NiagaraLightRendererProperties]}
 ]
 userVariables [
  User.AnimElimKill {type=NiagaraFloat, defaultValue=0, description=""}
  User.bIsFrontEnd {type=NiagaraBool}
  User.bIsFrontEndPreview {type=NiagaraBool}
 ]
}
